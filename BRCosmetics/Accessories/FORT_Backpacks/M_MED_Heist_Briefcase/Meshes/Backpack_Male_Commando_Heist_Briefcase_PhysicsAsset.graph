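# Physics asset for the Heist Briefcase back accessory on the male Commando, defining collision and joint limits for its dynamic bones.
# Three overlapping capsules cover backpack_root, and one capsule sits on each bone of the six-bone dyn_child chain, small in the middle and larger at both ends.
# Constraints link the chain in sequence from the root, with a tight swing at the first joint and wider swing, mainly on one axis, below it.
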PhysicsAsset /BRCosmetics/Accessories/FORT_Backpacks/M_MED_Heist_Briefcase/Meshes/Backpack_Male_Commando_Heist_Briefcase_PhysicsAsset
bodies {
 backpack_root [
  {
   shapeType=Capsule
   length=31.9654
   radius=8
   shapeName=backpack_root_capsule
   center=<-9.9085 -2.1303 -3.3309>
   rotation=<rot -1.5808 -90.0007 162.8091>
  }
  {shapeType=Capsule, length=33.25, radius=8, shapeName=backpack_root_capsule, center=<-4.74 -1.735 -1.7252>, rotation=<rot 20 -0.0004 0>}
  {shapeType=Capsule, length=32, radius=8, shapeName=backpack_root_capsule, center=<-0.6815 -1.2736 -0.248>, rotation=<rot 20 -0.0002 0>}
 ]
 dyn_child_01 [{
   shapeType=Capsule
   length=8.4176
   radius=3.6136
   shapeName=dyn_child_01_capsule
   center=<-0.6648 -1.834 -0.9654>
   rotation=<rot 0 175.665 -67.5706>
  }]
 dyn_child_02 [{shapeType=Capsule, length=0, radius=1, center=<0 0 -0.8255>}]
 dyn_child_03 [{shapeType=Capsule, length=0.25, radius=1, center=<0 0 -1.6119>}]
 dyn_child_04 [{shapeType=Capsule, length=0.25, radius=1, center=<0 0 -2.7768>}]
 dyn_child_05 [{
   shapeType=Capsule
   length=5.3779
   radius=4.2368
   shapeName=dyn_child_05_capsule
   center=<-0.5066 -2.4303 -7.0796>
   rotation=<rot -0.0846 -90 10.6307>
  }]
 dyn_child_06 [{
   shapeType=Capsule
   length=2.25
   radius=5.75
   shapeName=dyn_child_06_capsule
   center=<-0.6724 -3.6761 -4.8266>
   rotation=<rot 0.1235 -90 -4.4558>
  }]
}
constraints [
 {bone1Name=dyn_child_02, bone2Name=dyn_child_01, swing1LimitDegrees=10, swing2LimitDegrees=50, twistLimitDegrees=10}
 {bone1Name=dyn_child_04, bone2Name=dyn_child_03, swing1LimitDegrees=10, swing2LimitDegrees=50, twistLimitDegrees=10}
 {bone1Name=dyn_child_05, bone2Name=dyn_child_04, swing1LimitDegrees=10, swing2LimitDegrees=50, twistLimitDegrees=10}
 {bone1Name=dyn_child_06, bone2Name=dyn_child_05, swing1LimitDegrees=10, swing2LimitDegrees=50, twistLimitDegrees=10}
 {bone1Name=dyn_child_01, bone2Name=backpack_root, swing1LimitDegrees=2, swing2LimitDegrees=0, twistLimitDegrees=10}
 {bone1Name=dyn_child_03, bone2Name=dyn_child_02, swing1LimitDegrees=10, swing2LimitDegrees=50, twistLimitDegrees=10}
]
