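# Animation Blueprint for the Commando Heist Briefcase backpack cosmetic, adding simulated physics motion to its skeleton.
# A RigidBody node simulates the briefcase physics asset in backpack_root space over the input pose, with overridden gravity and a fixed half-strength blend.
# Each update lowers the node's linear acceleration scale while targeting or skydiving.
# The update also sets dynAlpha by front-end state, but the AnimGraph never reads it.
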AnimBlueprint /BRCosmetics/Accessories/FORT_Backpacks/M_MED_Heist_Briefcase/Meshes/Backpack_Male_Commando_Heist_Briefcase_Skeleton_AnimBP
target_skeleton=&/Backpack_Male_Commando_Heist_Briefcase_Skeleton
graphs [
 AnimGraph {nodes [
   n1 = RigidBody!(
    ComponentPose=Local To Component(LocalPose=Input Pose())
    Alpha=0.5
    ExternalForce
    ComponentLinearAccScale:Vector=Get override_acc_acale() -> (override_acc_acale:Vector)
   ) -> (Pose) {
    bClampLinearTranslationLimitToRefPose=true
    bOverrideWorldGravity=true
    overridePhysicsAsset=&/Backpack_Male_Commando_Heist_Briefcase_PhysicsAsset
    simulationSpace=BaseBoneSpace
    alphaBoolBlend {blendOption=Linear}
    baseBoneRef {boneName=backpack_root}
    componentAppliedLinearAccClamp=<600 600 600>
    componentLinearAccScale=<0.4 0.3 0.3>
    componentLinearVelScale=<0.2 0.2 0.2>
    overrideWorldGravity {z=-980}
   }
   Output Pose(Result=Component To Local(ComponentPose=@n1)) {layerGroup=None}
  ]}
 EventGraph {nodes [
   Event BlueprintUpdateAnimation()
   n1 = Get bIsTargeting()
   n2 = Get bIsSkyDiving()
   > Set override_acc_acale(override_acc_acale:Vector=SelectVector(A="0.05,0.05,0.05", B="0.3,0.3,0.35", bPickA=OR Boolean(A=@n1, B=@n2))) -> (Output_Get:Vector)
   > n3 = Branch(Condition=Get bIsInFrontEnd())
   > Set dynAlpha(dynAlpha:double=0.5) -> (Output_Get:double)
   Set dynAlpha(execute=@n3.else, dynAlpha=1)
  ]}
]
variables [
 dynAlpha {type=number}
 override_acc_acale {type=Vector, default=<0.3 0.3 0.35>}
]
